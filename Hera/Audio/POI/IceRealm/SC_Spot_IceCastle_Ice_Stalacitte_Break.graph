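# Sound cue for a spot ambient ice stalactite break sound in the Ice Castle point of interest.
# A Random node picks one of five equally weighted break wave variations, then a Modulator varies volume and lowers pitch by a random amount.
# The cue plays at reduced volume through a one-shot tile ambient attenuation and the indoor ambient Athena sound class.
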SoundCue /Hera/Audio/POI/IceRealm/SC_Spot_IceCastle_Ice_Stalacitte_Break
root=Modulator_0
nodes [
 {id=Modulator_0, children [Random_0], properties {pitchMax=1, pitchMin=0.8, volumeMax=1, volumeMin=0.7}}
 {id=Random_0, children [WavePlayer_3,WavePlayer_4,WavePlayer_5,WavePlayer_6,WavePlayer_7], properties {weights [1,1,1,1,1]}}
 {id=WavePlayer_3, properties {soundWaveAssetPtr=/AmbientAudioContent/Waves/SpottedLoops/Spot_IceCastle_Ice_Stalacitte_Break_01}}
 {id=WavePlayer_4, properties {soundWaveAssetPtr=/AmbientAudioContent/Waves/SpottedLoops/Spot_IceCastle_Ice_Stalacitte_Break_02}}
 {id=WavePlayer_5, properties {soundWaveAssetPtr=/AmbientAudioContent/Waves/SpottedLoops/Spot_IceCastle_Ice_Stalacitte_Break_03}}
 {id=WavePlayer_6, properties {soundWaveAssetPtr=/AmbientAudioContent/Waves/SpottedLoops/Spot_IceCastle_Ice_Stalacitte_Break_04}}
 {id=WavePlayer_7, properties {soundWaveAssetPtr=/AmbientAudioContent/Waves/SpottedLoops/Spot_IceCastle_Ice_Stalacitte_Break_05}}
]
properties {
 attenuationSettings=/Game/Sounds/Attenuation/Fort_TileAmbient_OneShot_Att
 maxDistance=8000
 soundClassObject=&Fort_Ambient_Indoors_Athena
 volumeMultiplier=0.5
}
