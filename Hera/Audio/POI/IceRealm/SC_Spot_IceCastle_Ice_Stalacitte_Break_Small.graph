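# Sound cue for a small ice stalactite break, an ambient spot sound for the Ice Castle in the Ice Realm POI.
# A Random node picks evenly among ten break wave variations, and a Modulator slightly varies pitch and volume to reduce repetition.
# It plays quietly on the Athena indoor ambient sound class with a tile-ambient one-shot attenuation.
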SoundCue /Hera/Audio/POI/IceRealm/SC_Spot_IceCastle_Ice_Stalacitte_Break_Small
root=Modulator_0
nodes [
 {id=Modulator_0, children [Random_0], properties {pitchMax=1, pitchMin=0.8, volumeMax=1, volumeMin=0.7}}
 {
  id=Random_0
  children [
   WavePlayer_0
   WavePlayer_1
   WavePlayer_2
   WavePlayer_8
   WavePlayer_9
   WavePlayer_10
   WavePlayer_11
   WavePlayer_12
   WavePlayer_13
   WavePlayer_14
  ]
  properties {weights [1,1,1,1,1,1,1,1,1,1]}
 }
 {id=WavePlayer_0, properties {soundWaveAssetPtr=/AmbientAudioContent/Waves/SpottedLoops/Spot_IceCastle_Ice_Stalacitte_Break_Small_01}}
 {id=WavePlayer_1, properties {soundWaveAssetPtr=/AmbientAudioContent/Waves/SpottedLoops/Spot_IceCastle_Ice_Stalacitte_Break_Small_02}}
 {id=WavePlayer_10, properties {soundWaveAssetPtr=/AmbientAudioContent/Waves/SpottedLoops/Spot_IceCastle_Ice_Stalacitte_Break_Small_06}}
 {id=WavePlayer_11, properties {soundWaveAssetPtr=/AmbientAudioContent/Waves/SpottedLoops/Spot_IceCastle_Ice_Stalacitte_Break_Small_07}}
 {id=WavePlayer_12, properties {soundWaveAssetPtr=/AmbientAudioContent/Waves/SpottedLoops/Spot_IceCastle_Ice_Stalacitte_Break_Small_08}}
 {id=WavePlayer_13, properties {soundWaveAssetPtr=/AmbientAudioContent/Waves/SpottedLoops/Spot_IceCastle_Ice_Stalacitte_Break_Small_09}}
 {id=WavePlayer_14, properties {soundWaveAssetPtr=/AmbientAudioContent/Waves/SpottedLoops/Spot_IceCastle_Ice_Stalacitte_Break_Small_10}}
 {id=WavePlayer_2, properties {soundWaveAssetPtr=/AmbientAudioContent/Waves/SpottedLoops/Spot_IceCastle_Ice_Stalacitte_Break_Small_03}}
 {id=WavePlayer_8, properties {soundWaveAssetPtr=/AmbientAudioContent/Waves/SpottedLoops/Spot_IceCastle_Ice_Stalacitte_Break_Small_04}}
 {id=WavePlayer_9, properties {soundWaveAssetPtr=/AmbientAudioContent/Waves/SpottedLoops/Spot_IceCastle_Ice_Stalacitte_Break_Small_05}}
]
properties {
 attenuationSettings=/Game/Sounds/Attenuation/Fort_TileAmbient_OneShot_Att
 maxDistance=8000
 soundClassObject=&Fort_Ambient_Indoors_Athena
 volumeMultiplier=0.25
}
